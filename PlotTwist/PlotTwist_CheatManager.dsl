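; Cheat manager for the PlotTwist feature, offering commands that strip Force powers or lightsabers from the player.
; Each function loops over a list of item definitions, Force power augments or lightsaber items, and removes every entry from the player controller's inventory.
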(fn PlotTwist_RemoveForcePowers
  (for _array_element (Variables|Default|GetForcePowerAugments)
    (Inventory|RemoveItemFromPlayer (CheatManager|GetPlayerController) _array_element)))

(fn PlotTwist_RemoveLightsabers
  (for _array_element (Variables|Default|GetLightsaberItemDefinitions)
    (Inventory|RemoveItemFromPlayer (CheatManager|GetPlayerController) _array_element)))
