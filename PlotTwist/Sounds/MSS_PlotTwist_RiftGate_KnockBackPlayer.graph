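# One-shot stereo MetaSound for the PlotTwist Rift Gate player knockback sound.
# On Play triggers a random stereo wave player patch fed by a wave asset array, which runs through a stereo mixer to the outputs and signals OnFinished.
# It uses its own attenuation asset, the Fort_Gadgets sound class and a concurrency override that caps simultaneous instances.
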MetaSoundSource /PlotTwist/Sounds/MSS_PlotTwist_RiftGate_KnockBackPlayer
inputs [UE.Source.OnPlay {type=Trigger}]
interfaces [UE.OutputFormat.Stereo:1.0, UE.Source.OneShot:1.0, UE.Source:1.1]
outputs [
 "UE.OutputFormat.Stereo.Audio:0" {type=Audio}
 "UE.OutputFormat.Stereo.Audio:1" {type=Audio}
 UE.Source.OneShot.OnFinished {type=Trigger}
]
pages [Default {
  input_defaults {UE.Source.OnPlay=null}
  nodes [
   MSP_Audio_StereoRng_Waveplayer_0 = /FortMetasound/Patches/MSP_Audio_StereoRng_Waveplayer(Play=^UE.Source.OnPlay, In Array=VariableAccessor.WaveAsset:Array(Variable=InitVariable.WaveAsset:Array().Variable).Value) {also {defaults {Enable Shared State="<MetaSoundBuilderResult.SUCCEEDED: 0>"}}}
   AudioMixerAudioMixerStereo2_0 = "AudioMixer.Audio Mixer (Stereo, 2)"(In 0 L=@MSP_Audio_StereoRng_Waveplayer_0.Out.L, In 0 R=@MSP_Audio_StereoRng_Waveplayer_0.Out.R)
  ]
  links [
   {dst="@UE.OutputFormat.Stereo.Audio:0", src=AudioMixerAudioMixerStereo2_0.Out L}
   {dst="@UE.OutputFormat.Stereo.Audio:1", src=AudioMixerAudioMixerStereo2_0.Out R}
   {dst="@UE.Source.OneShot.OnFinished", src=MSP_Audio_StereoRng_Waveplayer_0.On Finished}
  ]
 }]
properties {
 assetImportData=&self:AssetImportData
 attenuationSettings=&/ATT_PlotTwist_RiftGate_KnockBackPlayer
 bOverrideConcurrency=true
 maxDistance=6500
 numChannels=2
 outputFormat=Stereo
 sampleRate=48000
 soundClassObject=&Fort_Gadgets
 concurrencyOverrides {maxCount=3}
}
