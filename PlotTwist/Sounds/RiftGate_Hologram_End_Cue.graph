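# Sound cue for the Rift Gate hologram end sound, in the gadgets and items sound class.
# A single wave player feeds one wave asset straight to the output, with no randomization or mixing.
# Cue-level attenuation overrides take precedence over the referenced Rift Gate attenuation asset, setting falloff and low-pass ranges and disabling the reverb send.
# Overridden concurrency allows one instance at a time and prevents new plays while it is active.
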SoundCue /PlotTwist/Sounds/RiftGate_Hologram_End_Cue
root=WavePlayer_0
nodes [{id=WavePlayer_0, properties {soundWaveAssetPtr=&/RiftGate_Hologram_End_01}}]
properties {
 attenuationSettings=&/Attenuation_RiftGate
 bOverrideAttenuation=true
 bOverrideConcurrency=true
 maxDistance=1100
 soundClassObject=&Fort_Gadgets_Items
 volumeMultiplier=0.8
 attenuationOverrides {
  bEnableReverbSend=false
  falloffDistance=800
  lPFFrequencyAtMax=2000
  lPFRadiusMax=2000
  lPFRadiusMin=1000
  reverbWetLevelMax=0.7
  stereoSpread=100
  attenuationShapeExtents {x=300}
 }
 concurrencyOverrides {maxCount=1, resolutionRule=PreventNew}
}
