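# Sound cue that plays the looping hologram sound for the Rift Gate, routed to the gadgets/items sound class at reduced volume.
# A single looping wave player feeds the output directly, with no randomization or mixing.
# Inline attenuation overrides replace the referenced Rift Gate attenuation asset, setting a short falloff range with reverb send disabled.
# A concurrency override caps it at two simultaneous instances.
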SoundCue /PlotTwist/Sounds/RiftGate_Hologram_Loop_Cue
root=WavePlayer_3
nodes [{id=WavePlayer_3, properties {bLooping=true, soundWaveAssetPtr=&/RiftGate_Hologram_Loop_01}}]
properties {
 attenuationSettings=&/Attenuation_RiftGate
 bOverrideAttenuation=true
 bOverrideConcurrency=true
 maxDistance=1250
 soundClassObject=&Fort_Gadgets_Items
 volumeMultiplier=0.5
 attenuationOverrides {
  bEnableReverbSend=false
  falloffDistance=1200
  lPFFrequencyAtMax=2000
  lPFRadiusMax=2000
  lPFRadiusMin=1000
  reverbWetLevelMax=0.7
  stereoSpread=100
  attenuationShapeExtents {x=50}
 }
 concurrencyOverrides {bVolumeScaleCanRelease=true, maxCount=2}
}
