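# Sound cue for the RiftGate orb, combining a one-shot hologram start sound with a looping orb sound.
# A mixer layers the pitched-down start wave with the loop, which a short fixed delay holds back so it enters just after the start.
# It plays in the gadgets and items sound class, with overridden attenuation adding distance low-pass filtering and a concurrency override capping simultaneous instances.
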SoundCue /PlotTwist/Sounds/RiftGate_Orb_Loop_Cue
root=Mixer_0
nodes [
 {id=Delay_0, children [WavePlayer_1], properties {delayMax=0.2, delayMin=0.2}}
 {id=Mixer_0, children [Delay_0,Modulator_0], properties {inputVolume [1,1]}}
 {id=Modulator_0, children [WavePlayer_3], properties {pitchMax=0.8, pitchMin=0.8}}
 {id=WavePlayer_1, properties {bLooping=true, soundWaveAssetPtr=&/RiftGate_Orb_Loop_01}}
 {id=WavePlayer_3, properties {soundWaveAssetPtr=&/RiftGate_Hologram_Start_01}}
]
properties {
 attenuationSettings=&/Attenuation_RiftGate
 bOverrideAttenuation=true
 bOverrideConcurrency=true
 maxDistance=2500
 soundClassObject=&Fort_Gadgets_Items
 volumeMultiplier=0.7
 attenuationOverrides {
  bAttenuateWithLPF=true
  falloffDistance=2200
  lPFFrequencyAtMax=2000
  lPFRadiusMax=2200
  lPFRadiusMin=1000
  reverbWetLevelMax=0.7
  stereoSpread=1000
  attenuationShapeExtents {x=300}
 }
 concurrencyOverrides {bVolumeScaleCanRelease=true, maxCount=2}
}
