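# Sound cue for entering a RiftGate rift in PlotTwist.
# A single wave player feeds the RiftGate_Rift_Enter_01 sound directly to the output, with a slight volume boost.
# It uses the RiftGate attenuation and the Fort gadgets item sound class.
# A concurrency override allows only one instance at a time and blocks new plays while one is active.
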SoundCue /PlotTwist/Sounds/RiftGate_Rift_Enter_Cue
root=WavePlayer_1
nodes [{id=WavePlayer_1, properties {soundWaveAssetPtr=&/RiftGate_Rift_Enter_01}}]
properties {
 attenuationSettings=&/Attenuation_RiftGate
 bOverrideConcurrency=true
 maxDistance=5400
 soundClassObject=&Fort_Gadgets_Items
 volumeMultiplier=1.1
 concurrencyOverrides {maxCount=1, resolutionRule=PreventNew}
}
