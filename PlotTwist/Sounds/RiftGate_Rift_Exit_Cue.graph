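# Sound cue that plays the exit sound for the RiftGate rift, classed as a gadget or item sound.
# A single wave player feeds the RiftGate_Rift_Exit_01 wave straight to the output with boosted volume, using the shared RiftGate attenuation.
# A concurrency override allows one instance at a time and blocks new plays while it is active, and the cue primes on load.
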SoundCue /PlotTwist/Sounds/RiftGate_Rift_Exit_Cue
root=WavePlayer_2
nodes [{id=WavePlayer_2, properties {soundWaveAssetPtr=&/RiftGate_Rift_Exit_01}}]
properties {
 attenuationSettings=&/Attenuation_RiftGate
 bOverrideConcurrency=true
 bPrimeOnLoad=true
 maxDistance=5400
 soundClassObject=&Fort_Gadgets_Items
 volumeMultiplier=2
 concurrencyOverrides {maxCount=1, resolutionRule=PreventNew}
}
