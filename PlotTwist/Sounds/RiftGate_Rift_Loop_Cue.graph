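# Sound cue that plays the looping rift sound for the Rift Gate as an in-game SFX.
# A single looping wave player feeds the output directly, with boosted volume.
# Inline attenuation overrides the referenced Rift Gate attenuation asset, adding distance-based low-pass filtering and stereo spread.
# The override also tunes reverb send levels and interpolated listener focus, which quiets the loop when it is out of focus.
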SoundCue /PlotTwist/Sounds/RiftGate_Rift_Loop_Cue
root=WavePlayer_0
nodes [{id=WavePlayer_0, properties {bLooping=true, soundWaveAssetPtr=&/RiftGate_Rift_Loop_01}}]
properties {
 attenuationSettings=&/Attenuation_RiftGate
 bOverrideAttenuation=true
 maxDistance=7400
 soundClassObject=&Fort_SFX_InGame
 volumeMultiplier=2
 attenuationOverrides {
  bAttenuateWithLPF=true
  bEnableFocusInterpolation=true
  falloffDistance=7000
  focusAttackInterpSpeed=0.5
  focusAzimuth=15
  focusDistanceScale=0.5
  focusReleaseInterpSpeed=0.25
  lPFFrequencyAtMax=2000
  lPFRadiusMax=7000
  lPFRadiusMin=2000
  nonFocusAzimuth=90
  nonFocusVolumeAttenuation=0.1
  nonSpatializedRadiusEnd=200
  reverbWetLevelMax=0.8
  reverbWetLevelMin=0.4
  stereoSpread=500
  customAttenuationCurve {editorCurveData {keys [<at 0 1>,<at 0.3 1>,<at 0.4 0.5>,<at 1 0>]}}
 }
}
