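# Sound cue for the Rift Gate rift opening, routed to the gadgets and items sound class with Rift Gate attenuation.
# A mixer plays the rift spawn wave immediately and layers in the rift opens wave after a fixed one-second delay.
# A concurrency override allows one instance at a time and prevents new ones from starting while it plays.
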SoundCue /PlotTwist/Sounds/RiftGate_Rift_Opens_Cue
root=Mixer_0
nodes [
 {id=Delay_0, children [WavePlayer_0], properties {delayMax=1, delayMin=1}}
 {id=Mixer_0, children [WavePlayer_1,Delay_0], properties {inputVolume [1,1]}}
 {id=WavePlayer_0, properties {soundWaveAssetPtr=&/RiftGate_Rift_Opens_01}}
 {id=WavePlayer_1, properties {soundWaveAssetPtr=&/RiftGate_Rift_Spawn_01}}
]
properties {
 attenuationSettings=&/Attenuation_RiftGate
 bOverrideConcurrency=true
 maxDistance=5400
 soundClassObject=&Fort_Gadgets_Items
 volumeMultiplier=0.9
 concurrencyOverrides {maxCount=1, resolutionRule=PreventNew}
}
